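; Fortnite guided-motion movement mode for the Zero Point Dash, overriding dash direction, travel delta and constant-force strength.
; Horizontal direction follows normalized acceleration beyond an angle deadzone, otherwise pawn forward; Z follows net-synced aim, optionally remapped and floored by reverse angle.
; Delta scales that direction's XY by aim's horizontal length, normalizes, then applies the configured distance.
; Strength is distance over max duration; delta and strength fall back to parent results when zero.
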(fn GetDirection_ConstantForce
  (params (GuidedMotionRuntimeData :type Fort%20Movement%20Mode%20Guided%20Motion%20Runtime%20Data%20Object%20Reference))
  (return (DashDirectionConfig|GetMovementDirection GuidedMotionRuntimeData)))

(fn GetMovementDelta
  (params (GuidedMotionRuntimeData :type Fort%20Movement%20Mode%20Guided%20Motion%20Runtime%20Data%20Object%20Reference))
  (bind _returnvalue (|CallParentFunction GuidedMotionRuntimeData))
  (Utilities|IsValid GuidedMotionRuntimeData
    (:Is%20Valid
      (return (Math|Vector|SelectVector _returnvalue (* (Math|Vector|Normalize (* (Math|Vector|MakeVector (Math|Vector|VectorLengthXY (Math|Rotator|GetRotationXVector (CharacterMovement|GetNetSyncedBaseAimRotation (Variables|FortMovementMode_BaseExtLogic|GetOwnerMovementComponent)))) (Math|Vector|VectorLengthXY (Math|Rotator|GetRotationXVector (CharacterMovement|GetNetSyncedBaseAimRotation (Variables|FortMovementMode_BaseExtLogic|GetOwnerMovementComponent)))) 1.0) (DashDirectionConfig|GetMovementDirection (Variables|GetGuidedMotionRuntimeData)))) (ScalableFloat|GetCachedScalableFloatValue (Variables|DashDistanceConfig|GetMovementDeltaDistance))) (== 0 (Math|Vector|VectorLengthSquared (* (Math|Vector|Normalize (* (Math|Vector|MakeVector (Math|Vector|VectorLengthXY (Math|Rotator|GetRotationXVector (CharacterMovement|GetNetSyncedBaseAimRotation (Variables|FortMovementMode_BaseExtLogic|GetOwnerMovementComponent)))) (Math|Vector|VectorLengthXY (Math|Rotator|GetRotationXVector (CharacterMovement|GetNetSyncedBaseAimRotation (Variables|FortMovementMode_BaseExtLogic|GetOwnerMovementComponent)))) 1.0) (DashDirectionConfig|GetMovementDirection (Variables|GetGuidedMotionRuntimeData)))) (ScalableFloat|GetCachedScalableFloatValue (Variables|DashDistanceConfig|GetMovementDeltaDistance))))))))
    (:Is%20Not%20Valid
      (return _returnvalue))))

(fn GetMovementDirection
  (params (GuidedMotionRuntimeData :type Fort%20Movement%20Mode%20Guided%20Motion%20Runtime%20Data%20Object%20Reference))
  (bind (_returnvalue_x _returnvalue_y) (Math|Vector|BreakVector (Math|Vector|SelectVector (Math|Vector|Normalize (LocalVariables|GetCurrentAcceleration)) (DashDirectionConfig|GetMovementDirectionXYwhenNotUsingAccelerationforTarget) (LocalVariables|GetbUseAccelerationForTarget))))
  (bind (_aimdirection_x _aimdirection_y _aimdirection_z) (Math|Vector|BreakVector (Variables|GetAimDirection_X)))
  (bind (_aimdirection_x_2 _aimdirection_y_2 _aimdirection_z_2) (Math|Vector|BreakVector (Variables|GetAimDirection_X)))
  (if (Math|Boolean|ANDBoolean (Utilities|IsValid GuidedMotionRuntimeData :!class /Script/Engine.KismetSystemLibrary) (Utilities|IsValid (Variables|FortMovementMode_BaseExtLogic|GetOwnerPawn) :!class /Script/Engine.KismetSystemLibrary) (Utilities|IsValid (Variables|FortMovementMode_BaseExtLogic|GetOwnerMovementComponent) :!class /Script/Engine.KismetSystemLibrary))
    (Utilities|FlowControl|Sequence
      (:then_0
        (LocalVariables|SetPawnForward (Transformation|GetActorForwardVector (Variables|FortMovementMode_BaseExtLogic|GetOwnerPawn)))
        (LocalVariables|SetAimDirection (Math|Rotator|GetRotationXVector (CharacterMovement|GetNetSyncedBaseAimRotation (Variables|FortMovementMode_BaseExtLogic|GetOwnerMovementComponent))))
        (LocalVariables|SetCurrentAcceleration (Pawn|Components|CharacterMovement|GetCurrentAcceleration (Variables|FortMovementMode_BaseExtLogic|GetOwnerMovementComponent))))
      (:then_1
        (bind _returnvalue (Fortnite|GetAngleDegrees (LocalVariables|GetPawnForward) (LocalVariables|GetCurrentAcceleration)))
        (LocalVariables|SetMovementAngle _returnvalue)
        (LocalVariables|SetbUseAccelerationForTarget (and (not (Math|Vector|VectorIsNearlyZero (LocalVariables|GetCurrentAcceleration))) (> (LocalVariables|GetMovementAngle) (ScalableFloat|GetCachedScalableFloatValue (Variables|DashDirectionConfig|GetUseMovementForTargetDeadzone))))))
      (:then_2
        (return _returnvalue_x _returnvalue_y (Math|Float|SelectFloat (Math|Float|Max(Float) (* (Math|Float|MapRangeClamped (LocalVariables|GetMovementAngle) (ScalableFloat|GetCachedScalableFloatValue (Variables|DashDirectionConfig|GetDynamicZMinReverseAngle)) (ScalableFloat|GetCachedScalableFloatValue (Variables|DashDirectionConfig|GetDynamicZMaxReverseAngle)) 1.0 (ScalableFloat|GetCachedScalableFloatValue (Variables|DashDirectionConfig|GetDynamicZReverseModifier))) _aimdirection_z) (ScalableFloat|GetCachedScalableFloatValue (Variables|DashDirectionConfig|GetDynamicZReverseTargetMinZ))) _aimdirection_z_2 (> (ScalableFloat|GetCachedScalableFloatValue (Variables|DashDirectionConfig|GetbUseDynamicZTarget)) 0)))))
    (else
      (return "0, 0, 0"))))

(fn GetMovementDirectionXY_WhenNotUsingAccelerationForTarget
  (return (Transformation|GetActorForwardVector (Variables|FortMovementMode_BaseExtLogic|GetOwnerPawn))))

(fn GetStrength_ConstantForce
  (params (GuidedMotionRuntimeData :type Fort%20Movement%20Mode%20Guided%20Motion%20Runtime%20Data%20Object%20Reference))
  (bind _returnvalue (|CallParentFunction GuidedMotionRuntimeData))
  (Utilities|IsValid GuidedMotionRuntimeData
    (:Is%20Valid
      (return (Math|Float|SelectFloat _returnvalue (* (ScalableFloat|GetCachedScalableFloatValue (Variables|DashDistanceConfig|GetMovementDeltaDistance)) (Math|Float|SafeDivide 1.0 (Variables|/Script/FortniteGame.FortMovementMode_GuidedMotionRuntimeData|GetLocalMaxDuration (Variables|GetGuidedMotionRuntimeData)))) (== (* (ScalableFloat|GetCachedScalableFloatValue (Variables|DashDistanceConfig|GetMovementDeltaDistance)) (Math|Float|SafeDivide 1.0 (Variables|/Script/FortniteGame.FortMovementMode_GuidedMotionRuntimeData|GetLocalMaxDuration (Variables|GetGuidedMotionRuntimeData)))) 0))))
    (:Is%20Not%20Valid
      (return _returnvalue))))
