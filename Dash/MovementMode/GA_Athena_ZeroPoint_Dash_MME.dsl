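; Gameplay ability for the Zero Point dash, which commits a cooldown, pushes a camera mode and moves the pawn unless clambering.
; Clients and server-controlled bots apply a dash or guided-motion movement mode extension, while the server applies gameplay effects and broadcasts usage.
; The variant-selected dash cue is deferred one frame to split activation cost, with an end-ability fallback.
; The ability ends after the dash duration, backed by a failsafe timer.
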(event Ability|EventActivateAbility
  (|CallParentFunction)
  ;[0.06584 0.1198 0.02184] On Begin, immediately set the camera and commit the cooldown
  (Camera|PushCameraMode (Variables|Default|GetCamera%20Mode%20Class))
  (HUD|SetHUDElementVisibility (BlueprintContext|GetFortHUDContext) (Variables|Default|GetT_HudElementToHide) true "None")
  (bind _ability_committed (CallFunction|AttemptCommitAbility))
  (if _ability_committed
    (bind _asfort_player_pawn (Utilities|Casting|CastToFortPlayerPawn (Ability|GetAvatarActorfromActorInfo))
      (:then
        (bind _output_get (Variables|Default|SetOwning%20Pawn _asfort_player_pawn))
        (Utilities|IsValid _output_get
          (:Is%20Valid
            ;[0.224 0.2204 0.06917] Check for Clamber Tags and stop ability if clambering
            (CallFunction|WaitforClamberTags)
            (if (not (Variables|Default|GetbIsClambering))
              (CallFunction|ApplyDashMME)
              (Utilities|FlowControl|Sequence
                (:then_0
                  ;[0.15 0.06733 0.01429 0.5] After initial checks pass, start failsafe timer to ensure ability eventually gets ended
                  (bind _returnvalue (Utilities|Time|SetTimerbyEvent (EventDispatchers|CreateEvent :!event K2_EndAbility) (Variables|Default|GetFailsafe%20Delay)))
                  (Variables|Default|SetFailsafeTimerHandle _returnvalue))
                (:then_1
                  ;[0.03139 0.15 0.124 0.5] Apply GEs and broadcast on server
                  (if (Networking|IsServer)
                    (Ability|ApplyGameplayEffectToOwner "/Game/Athena/Items/ForagedItems/ZeroPoint/GE_Quest_AfterDashing10s.GE_Quest_AfterDashing10s_C")
                    (bind _returnvalue_2 (Ability|ApplyGameplayEffectToOwner "/Game/Athena/Items/ForagedItems/ZeroPoint/GE_ZeroPoint_IsActive.GE_ZeroPoint_IsActive_C"))
                    (Variables|Default|SetGE_IsActive _returnvalue_2)
                    (CustomVerb|BroadcastCustomVerbMessage (CustomVerb|MakeCustomVerbMessage (Variables|Default|GetOwning%20Pawn) "(TagName=\"CustomVerb.ZeroPointDash.Used\")"))))
                (:then_2
                  ;[0.1321 0.06177 0.15 0.5] Add dash GC next frame to split activation cost
                  (Utilities|FlowControl|DelayUntilNextFrame)
                  (CallFunction|PlayDashFX))
                (:then_3
                  ;[0.06939 0.08954 0.15 0.5] Reset fall height and end ability after dash finishes
                  (Pawn|Character|ResetFallingHeight (Variables|Default|GetOwning%20Pawn))
                  (Utilities|FlowControl|Delay (+ (ScalableFloat|GetValueatLevel (Class|GetClassDefaults (Variables|Default|GetDash%20MME%20Class))) 0.1))
                  (Ability|EndAbility)))))
          (:Is%20Not%20Valid
            _))
        (Ability|EndAbility))
      (:CastFailed
        _))
    (else
      _))
  (Ability|EndAbility))

(event Custom|WaitForClamberTags
  (bind _t_clambering (Variables|Default|GetT_Clambering))
  ;[0.01509 0.1667 0.103] If the player is in clambering range, stop the montage and update our variable so we can prevent spawning a crashpad.
  (Ability|Tasks|WaitGameplayTagAdd _t_clambering
    (:then
      (Ability|Tasks|WaitGameplayTagRemove _t_clambering
        (:Removed
          (Variables|Default|SetbIsClambering false))))
    (:Added
      (Variables|Default|SetbIsClambering true)
      (Ability|Animation|MontageStop -1.0))))

(event Custom|Apply%20Dash%20MME
  (bind _owning_pawn (Variables|Default|GetOwning%20Pawn))
  ;[0.0511 0.05013 0.15 0.5] Apply Dash MME on clients and server controlled AI
  (if (or (not (Networking|IsServer)) (Pawn|IsBotControlled (Variables|Default|GetOwning%20Pawn)))
    (if (ScalableFloat|GetValueasBoolwithContext (Variables|Default|GetUseGuidedMotion))
      (Game|ApplyGuidedMotionMME _owning_pawn (Variables|Default|GetGuided%20Motion%20MME%20Class))
      (else
        (Class|FortMovementModeExtDash|TryActivateDashMME _owning_pawn (Variables|Default|GetDash%20MME%20Class) (Utilities|Struct|MakeFortMovementModeDashCreationData (Targeting|GetDefaultTargetingRotation (Variables|Default|GetOwning%20Pawn)) (not (Math|Vector|VectorIsNearlyZero (Pawn|Components|PawnMovement|GetLastInputVector (Variables|/Script/Engine.Character|GetCharacterMovement (Variables|Default|GetOwning%20Pawn)))))))))))

(event Custom|Play%20Dash%20FX
  ;[0.0984 0.05539 0.15 0.5] Execute Dash Gameplay Cue. Delayed until after activation to split activation frame cost
  (Utilities|FlowControl|DoOnce false
    (:Completed
      (bind _owning_pawn (Variables|Default|GetOwning%20Pawn)
        (:then
          (Ability|ExecuteGameplayCueOnOwner (select (GameplayTags|HasMatchingGameplayTag _owning_pawn "(TagName=\"Granted.Athena.Sprites.ZeroPoint.UseAlternativeDashEffects\")") (Variables|Default|GetSuperDashGCCue) (select (GameplayTags|HasMatchingGameplayTag _owning_pawn "(TagName=\"Granted.Athena.StyxDash.UseAlternateEffects\")") (Variables|Default|GetStyxDashGCCue) (Variables|Default|GetGC_Cue)))))
        (:else)))))

(event Ability|EventOnEndAbility
  (params (bWasCancelled :type Boolean))
  ;[0.1458 0.05048 0.02221] End Ability
  (|CallParentFunction bWasCancelled)
  (HUD|SetHUDElementVisibility (BlueprintContext|GetFortHUDContext) (Variables|Default|GetT_HudElementToHide) true "None")
  (CallFunction|PlayDashFX) ; It's possible for ability to end the frame after being activated. \nDash FX is delayed a frame, so make sure it gets \nexecuted before the ability ends
  (Ability|RemoveGameplayEffectFromOwnerWithHandle (Variables|Default|GetGE_IsActive))
  (Utilities|Time|ClearandInvalidateTimerbyHandle (Variables|Default|GetFailsafeTimerHandle)))

(fn AttemptCommitAbility
  ;[1 0.8122 0.2487] Handle commit ability in function to allow overriding cost type
  (bind _returnvalue (Ability|CommitAbilityCooldown false))
  (return _returnvalue))

(fn ImpactEnemy)

(fn OverrideFailedReason
  (params (FailedReason :type Gameplay%20Tag%20Container%20Structure%20%28by%20ref%29))
  (if (Math|Boolean|ORBoolean (and (and (GameplayTags|HasTag FailedReason "(TagName=\"MovementMode.Skydiving\")" true) (GameplayTags|HasTag FailedReason "(TagName=\"Activate.Failed.Tags.Blocked\")" true)) (ScalableFloat|GetValueasBool (Variables|Default|GetUseFailedOverrideOnSkydive))) (and (and (GameplayTags|HasTag FailedReason "(TagName=\"Activate.Failed.Tags.Blocked\")" true) (GameplayTags|HasTag FailedReason "(TagName=\"Gameplay.Action.Player.AllowGliderRedeploy\")" true)) (ScalableFloat|GetValueasBool (Variables|Default|GetUseFailedOverrideOnAllowGlider))) (and (and (GameplayTags|HasTag FailedReason "(TagName=\"Activate.Failed.Tags.Blocked\")" true) (GameplayTags|HasTag FailedReason "(TagName=\"MovementMode.Parachuting\")" true)) (ScalableFloat|GetValueasBool (Variables|Default|GetUseFailedOverrideOnParachute))))
    (return "(GameplayTags=((TagName=\"Athena.Ability.Failed.NoMessage\")))")
    (else
      (return FailedReason))))

(fn SetPawnCollision
  (params (Condition :type Boolean))
  (bind _owning_pawn (Variables|Default|GetOwning%20Pawn))
  (bind _capsulecomponent (Variables|/Script/Engine.Character|GetCapsuleComponent _owning_pawn))
  (bind _owning_pawn_2 (Variables|Default|GetOwning%20Pawn))
  (bind _owning_pawn_3 (Variables|Default|GetOwning%20Pawn))
  (bind _capsulecomponent_2 (Variables|/Script/Engine.Character|GetCapsuleComponent _owning_pawn_2))
  (bind _capsulecomponent_3 (Variables|/Script/Engine.Character|GetCapsuleComponent _owning_pawn_3))
  (if Condition
    (if (Utilities|Name|NotEqual(Name) (Variables|Default|GetCollision%20Profile) "None")
      (Collision|SetCollisionProfileName _capsulecomponent_2 (Variables|Default|GetCollision%20Profile) true :!class /Script/Engine.PrimitiveComponent)
      (Collision|SetCollisionResponsetoChannel _capsulecomponent_3 "ECC_Pawn" "ECR_Block")
      (Collision|SetCollisionResponsetoChannel _capsulecomponent_3 "ECC_GameTraceChannel4" "ECR_Block"))
    (else
      (Variables|Default|SetCollision%20Profile (Collision|GetCollisionProfileName _capsulecomponent_2))
      (Collision|SetCollisionResponsetoChannel _capsulecomponent "ECC_Pawn" "ECR_Overlap")
      (Collision|SetCollisionResponsetoChannel _capsulecomponent "ECC_GameTraceChannel4" "ECR_Overlap"))))
